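# Sound cue for a disguise gadget transformation stinger, named for the Mando-side Force variant, with attenuation and single-instance concurrency that blocks new plays.
# An IsLocal branch selects between two mixers, both layering a pitched-down disguise disappear sound with a randomly pitched-down triangle bounce, the bounce weighted louder.
# Only the local mixer adds a Fort_Music_NPC class slot, whose wave player has no wave assigned; false and unset share the other mixer.
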SoundCue /Game/Athena/Sounds/Gadgets/Disguise/Akita/IP_SC_HeroTrans_Stinger_MandoSide_Force_Cue
root=Branch_0
nodes [
 {id=Branch_0, children [Mixer_0,Mixer_1,Mixer_1], properties {boolParameterName=IsLocal}}
 {id=Mixer_0, children [SoundClass_0,Modulator_0,Modulator_1], properties {inputVolume [1,1,1.5]}}
 {id=Mixer_1, children [null,Modulator_0,Modulator_1], properties {inputVolume [1,1,2]}}
 {id=Modulator_0, children [WavePlayer_2], properties {pitchMax=0.6, pitchMin=0.6}}
 {id=Modulator_1, children [WavePlayer_3], properties {pitchMax=0.9, pitchMin=0.8}}
 {id=SoundClass_0, children [WavePlayer_1], properties {soundClassOverride=&Fort_Music_NPC}}
 {id=WavePlayer_1}
 {id=WavePlayer_2, properties {soundWaveAssetPtr=/Game/Athena/Sounds/Gadgets/Disguise/Item_Disguise_Disappear_01}}
 {id=WavePlayer_3, properties {soundWaveAssetPtr=&/Concert_Audio_TriangleBounce_03_AF}}
]
properties {
 attenuationSettings=/Game/Athena/Sounds/Gadgets/Disguise/Disguise_Equip_Att
 bOverrideConcurrency=true
 maxDistance=1712
 priority=1.5
 soundClassObject=&Fort_Music_NPC
 volumeMultiplier=1
 concurrencyOverrides {maxCount=1, resolutionRule=PreventNew}
 concurrencySet [/Game/Athena/Sounds/Gadgets/Disguise/Disguise_Concurrency]
}
